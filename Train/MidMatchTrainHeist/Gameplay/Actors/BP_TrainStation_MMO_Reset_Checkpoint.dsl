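; Train station checkpoint, a server-only child of the transit sign base, that resets the cargo car's MMO capture in the mid-match train heist.
; Once playlist data is ready, hotfix scalable floats destroy it if disabled or if only the primary location may reset and this is not it.
; It caches the cargo car, and when time until arrival nears zero, searches the car's child actors for the MMO interaction actor and resets any completed capture.
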(event Collision|EventActorBeginOverlap
  (params (OtherActor :type Actor%20Object%20Reference)))

(event EventTick
  (params (DeltaSeconds :type Float%20%28single-precision%29)))

(event EventBeginPlay
  ;[0.3746 0 1] Check hotfix values for enabled & if only 1 station will reset vs multiple
  (Class|FortAsyncActionCurrentPlaylistReady|PlaylistDataReadyAsync
    (:OnReady
      (if (ScalableFloat|GetValueasBool (Variables|Default|GetSF_Enabled))
        (if (ScalableFloat|GetValueasBool (Variables|Default|GetSF_PrimaryLocationOnly))
          (if (not (Variables|Default|GetPrimaryResetCheckpoint))
            (Actor|DestroyActor))
          (else
            _))
        (bind _returnvalue (Actor|GetActorOfClass "/Train/Cars/BaseCarCargoAlt_MMO_Version.BaseCarCargoAlt_MMO_Version_C"))
        (Variables|Default|SetHeliosCargoCar_MMO_Verstion _returnvalue)
        (else
          (Actor|DestroyActor))))))

(event EventOnArrivalTimeIntervalReachedForStation
  (params (TImeUntilArrival :type Float%20%28double-precision%29))
  ;[0 0.9112 1] [SERVER] MMO checkpoints are set to run as server only child of BP Base Transit Sign. Once arrival at station event fires, grab the MMO actor and reset it.
  (Networking|SwitchHasAuthority
    (:Authority
      (if (Math|Float|NearlyEqual(Float) TImeUntilArrival)
        (bind _array_element (Utilities|Array|ForEachLoopwithBreak (Actor|GetAllChildActors (Variables|Default|GetHeliosCargoCar_MMO_Verstion) true))
          (:LoopBody
            (bind _asbga_train_mmo_interaction_parent (Utilities|Casting|CastToBGA_Train_MMO_Interaction_Parent _array_element)))
          (:Completed
            _))
        (if (Variables|/Train/MidMatchTrainHeist/Gameplay/Actors/BGA_Train_MMO_Interaction_Parent.BGA_Train_MMO_Interaction_Parent_C|GetIsMMOCaptureComplete _asbga_train_mmo_interaction_parent)
          (Class|BGATrainMMOInteractionParent|ResetCapture _asbga_train_mmo_interaction_parent))))))

(fn OnRep_HeliosCargoCar_MMO_Verstion)

(fn ConstructionScript)
